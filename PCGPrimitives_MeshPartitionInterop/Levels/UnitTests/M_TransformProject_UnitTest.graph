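# Debug material for the TransformProject unit test in the PCG primitives Mesh Partition interop plugin, making a mesh partition channel visible as color.
# It samples the Test channel defined by MPD_TransformProject_UnitTest from the Mesh Partition resource's channel texture.
# That value drives a lerp from neutral grey to red, so red marks where the channel is set.
# The material is Default Lit and flagged for use with Nanite.
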Material /PCGPrimitives_MeshPartitionInterop/Levels/UnitTests/M_TransformProject_UnitTest
graph {
 properties {bUsedWithNanite=true, usedShadingModels=Default Lit}
 nodes [LinearInterpolate(
   A=Constant3Vector() {constant=<rgba 0.5104 0.5104 0.5104 0>}
   B=Constant3Vector() {constant=<rgba 1 0 0 0>}
   Alpha=MeshPartitionChannelSample!(ChannelTextureInput=MeshPartitionResource().Channel Texture) {channel=Test, meshPartitionDefinition=&/MPD_TransformProject_UnitTest}
  )]
}
