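; PCG unit test graph for the Transform_Project primitive, including its MeshTerrain interop.
; It projects a resampled spline, grid points and filtered MeshTerrain points onto a Polygon2D, a jittered spline and the MeshTerrain surface.
; Both sides carry a Test float to exercise attribute projection; spline results copy it to density for debug display.
; A tag filter routes projected terrain points to Write_MeshTerrain and other points to debug.
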(graph PCG_TransformProject_UnitTest
  (bind Subgraph_1 (Subgraph :Graph /PCGPrimitives/Primitives/Transform/Transform_Project.Transform_Project))
  (bind Subgraph_3 (Subgraph :Graph /PCGPrimitives/Primitives/Debug/Debug_PrimitiveData.Debug_PrimitiveData)) ; ViewTargets
  (bind FilterDataByType_2 (Filter%20Data%20By%20Type :TargetType (struct :Ids [(struct :Struct /Script/PCG.PCGDataTypeInfoSpline)])))
  (bind Subgraph_16 (Subgraph :Graph /PCGPrimitives/Primitives/Filter/Filter_Tags.Filter_Tags)) ; Only points from ExtractMeshTerrainPoints
  (bind FilterDataByType_4 (Filter%20Data%20By%20Type :TargetType (struct :Ids [(struct :Struct /Script/PCG.PCGDataTypeInfoPoint)])))
  (bind Subgraph_14 (Subgraph :Graph /PCGPrimitives_MeshPartitionInterop/Primitives/Write/Write_MeshTerrain.Write_MeshTerrain))
  (bind Debug_0 (Debug)) ; View Source Points
  (bind Subgraph_0 (Subgraph :Graph /PCGPrimitives/Primitives/Debug/Debug_PrimitiveData.Debug_PrimitiveData)) ; View SourceSpline
  (bind GetSplineControlPoints_0 (Get%20Spline%20Control%20Points))
  (bind CopyAttributes_0 (Copy%20Attributes :InputSource Test :OutputTarget $Density))
  (bind Debug_1 (Debug)) ; View SourceSpline
  ;[0.15 0.15 0.15 0.5] Target Polygon2D
  (bind Subgraph_2 (Subgraph :Graph /PCGPrimitives/Primitives/Create/Create_Spline_Shape.Create_Spline_Shape))
  (bind AddAttribute_1 (Add%20Attribute :AttributeTypes (struct :Type Float :FloatValue 0.5 :DoubleValue 1.0) :OutputTarget Test)) ; Test attrib
  ;[0.15 0.15 0.15 0.5] Source Spline
  (bind Subgraph_4 (Subgraph :Graph /PCGPrimitives/Primitives/Create/Create_Spline_line.Create_Spline_line))
  (bind Subgraph_6 (Subgraph :Graph /PCGPrimitives/Primitives/Transform/Transform_Resample.Transform_Resample))
  (bind AddAttribute_2 (Add%20Attribute :AttributeTypes (struct :Type Float) :OutputTarget Test)) ; Test ProjectAttribute will be added if missing on Source input
  ;[0.15 0.15 0.15 0.5] Target Spline
  (bind Subgraph_8 (Subgraph :Graph /PCGPrimitives/Primitives/Create/Create_Spline_line.Create_Spline_line))
  (bind Subgraph_10 (Subgraph :Graph /PCGPrimitives/Primitives/Transform/Transform_Resample.Transform_Resample))
  (bind Subgraph_11 (Subgraph :Graph /PCGPrimitives/Primitives/Transform/Transform_Jitter.Transform_Jitter))
  (bind Subgraph_9 (Subgraph :Graph /PCGPrimitives/Primitives/Transform/Transform_Resample.Transform_Resample))
  (bind AddAttribute_4 (Add%20Attribute :AttributeTypes (struct :Type Float :FloatValue 1.0) :OutputTarget Test)) ; Test attrib
  ;[0.15 0.15 0.15 0.5] Source Points
  (bind CreatePointsGrid_1 (Create%20Points%20Grid :GridExtents (5e+03 5e+02 5e+01)))
  (bind TransformPoints_1 (Transform%20Points :OffsetMax (0.0 2.5e+03 1e+03) :OffsetMin (0.0 2.5e+03 1e+03)))
  (bind AddAttribute_5 (Add%20Attribute :AttributeTypes (struct :Type Float :DoubleValue 1.0) :OutputTarget Test)) ; Test ProjectAttribute will be added if missing on Source input
  ;[0.15 0.15 0.15 0.5] Target MeshTerrain Surface
  (bind Subgraph_12 (Subgraph :Graph /PCGPrimitives_MeshPartitionInterop/Primitives/Get/Get_MeshTerrain.Get_MeshTerrain))
  ;[0.15 0.15 0.15 0.5] Source MeshTerrain
  (bind Subgraph_13 (Subgraph :Graph /PCGPrimitives_MeshPartitionInterop/Primitives/Extract/Extract_MeshTerrainPoints.Extract_MeshTerrainPoints))
  (bind Subgraph_15 (Subgraph :Graph /PCGPrimitives/Primitives/Filter/Filter_AttributeValue.Filter_AttributeValue))
  (bind AddAttribute_6 (Add%20Attribute :AttributeTypes (struct :Type Float) :OutputTarget Test)) ; Test ProjectAttribute will be added if missing on Source input
  (connect Subgraph_2 Out AddAttribute_1 In)
  (connect Subgraph_6 Out AddAttribute_2 In)
  (connect Subgraph_9 Out AddAttribute_4 In)
  (connect TransformPoints_1 Out AddAttribute_5 In)
  (connect Subgraph_15 Unselected AddAttribute_6 In)
  (connect GetSplineControlPoints_0 Out CopyAttributes_0 Source)
  (connect GetSplineControlPoints_0 Out CopyAttributes_0 Target)
  (connect FilterDataByType_4 InsideFilter Debug_0 In)
  (connect CopyAttributes_0 Out Debug_1 In)
  (connect Subgraph_1 Out FilterDataByType_2 In)
  (connect Subgraph_16 Unselected FilterDataByType_4 In)
  (connect FilterDataByType_2 InsideFilter GetSplineControlPoints_0 In)
  (connect FilterDataByType_2 InsideFilter Subgraph_0 PrimaryInput)
  (connect AddAttribute_2 Out Subgraph_1 PrimaryInput)
  (connect AddAttribute_5 Out Subgraph_1 PrimaryInput)
  (connect AddAttribute_6 Out Subgraph_1 PrimaryInput)
  (connect AddAttribute_1 Out Subgraph_1 SecondaryInput)
  (connect AddAttribute_4 Out Subgraph_1 SecondaryInput)
  (connect Subgraph_12 Out Subgraph_1 SecondaryInput)
  (connect AddAttribute_1 Out Subgraph_3 PrimaryInput)
  (connect AddAttribute_4 Out Subgraph_3 PrimaryInput)
  (connect Subgraph_4 Out Subgraph_6 PrimaryInput)
  (connect Subgraph_11 Out Subgraph_9 PrimaryInput)
  (connect Subgraph_8 Out Subgraph_10 PrimaryInput)
  (connect Subgraph_10 Out Subgraph_11 PrimaryInput)
  (connect Subgraph_12 Out Subgraph_13 PrimaryInput)
  (connect Subgraph_16 Selected Subgraph_14 PrimaryInput)
  (connect Subgraph_13 Out Subgraph_15 PrimaryInput)
  (connect Subgraph_1 Out Subgraph_16 PrimaryInput)
  (connect CreatePointsGrid_1 Out TransformPoints_1 In))
